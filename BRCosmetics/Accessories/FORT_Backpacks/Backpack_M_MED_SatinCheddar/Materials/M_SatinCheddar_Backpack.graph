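# Skinned SatinCheddar backpack material that layers reactive frost, driven by a Velocity parameter, over the character parent shading.
# Parent PreFX attributes pass through the SatinCheddar ice function, then blend with a frost layer behind the UseReactivity switch before PostFX.
# A pre-skinned-position sphere mask and velocity shape the frost spread.
# Bump-offset Perlin noise, Fresnel and mask textures build frost color, snow edge, normal and time-of-day-adjusted emissive.
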Material /BRCosmetics/Accessories/FORT_Backpacks/Backpack_M_MED_SatinCheddar/Materials/M_SatinCheddar_Backpack
graph {
 properties {
  bUseMaterialAttributes=true
  bUsedWithMeshDeformer=true
  bUsedWithSkeletalMesh=true
  pixelDepthOffsetMode=PDOM_Legacy
  shadingModel=MSM_FromMaterialExpression
  usedShadingModels=Default Lit | Subsurface | Clear Coat | Subsurface Profile | Cloth
 }
 nodes [
  Constant() {r=0.99}
  n1 = MaterialFunctionCall() {materialFunction=&MF_CharacterParent_2023_PreFX}
  MaterialFunctionCall(Material=NamedRerouteUsage(), M=@n1.M, Pre Skinned Local Position=@n1.M, Distance Blend=@n1.M, Fresnel=@n1.M) {materialFunction=&MF_CharacterParent_2023_PostFX}
  n2 = Divide(A=Time()) {constB=10}
  Multiply(A=Abs(None=Sine(None=@n2) {desc=DebugReactivity}))
  n3 = Divide(A=Add(A=LocalPosition() {localOrigin=InstancePreSkinning}.XYZ, B=Constant3Vector() {constant=<rgba 0 0 5 0>})) {constB=100}
  NamedRerouteDeclaration(None=ShaderStageSwitch(PixelShader=VertexInterpolator(VS=@n3).PS, VertexShader=@n3)) {name=PreSkinnedPosition, nodeColor=<rgba 0.6875 0 1 1>}
  NamedRerouteDeclaration(None=MaterialFunctionCall!(In=@n1.Result, InMask, InFXMask=@n1.Result) {materialFunction=/BRCosmetics/Animation/Props/Gadgets_SatinCheddar/Materials/MF_SatinCheddar_Ice}.Result) {name=InMaterialAttributes, nodeColor=<rgba 0.4531 1 0 1>}
  n4 = GetMaterialAttributes(None=NamedRerouteUsage()) {attributeGetTypes [#1, #2]}
  NamedRerouteDeclaration(None=@n4.Base Color) {name=InBaseColor, nodeColor=<rgba 1 0 0.2344 1>}
  n5 = ComponentMask(None=NamedRerouteUsage()) {b=true, r=true}
  n6 = TextureSample(UVs=NamedRerouteDeclaration(None=Multiply(A=@n5) {constB=4}) {name=FrostUV, nodeColor=<rgba 0.7344 1 0 1>}) {samplerType=SAMPLERTYPE_Masks, texture=&/T_SatinCheddar_mask}
  NamedRerouteDeclaration(None=AppendVector(A=@n6.R, B=@n6.R)) {name=NormalXY, nodeColor=<rgba 0 1 0.1094 1>}
  n7 = Multiply(A=ComponentMask(None=NamedRerouteUsage()) {b=true, r=true}) {constB=0.25}
  n8 = TextureSample(UVs=BumpOffset(Coordinate=@n7, Height=Constant() {r=0.01}) {heightRatio=0.04}) {texture=/BRCosmetics/Animation/Props/Gadgets_SatinCheddar/Materials/T_Perlin_Noise_M1}
  n9 = Multiply(A=TextureSample(UVs=BumpOffset(Coordinate=@n7, Height=@n8.G) {heightRatio=0.01}) {texture=&T_Perlin_Noise_M1}.G) {constB=0.5}
  n10 = Transform(None=NamedRerouteUsage())
  n11 = Saturate(None=Subtract(A=NamedRerouteUsage(), B=Fresnel(Normal=@n10) {baseReflectFraction=0, exponent=0.3}))
  n12 = OneMinus(None=NamedRerouteUsage())
  n13 = Add(A=Multiply(A=Multiply(A=@n11, B=@n11)), B=Multiply(A=Power(Base=@n12) {constExponent=4}) {constB=0.1})
  n14 = LinearInterpolate(
   A=VectorParameter() {parameterName=ReactivityFrostColor, defaultValue=<rgba 0.0065 0.0443 0.0938 1>}.RGB
   B=VectorParameter() {parameterName=ReactivityFrostColor2, defaultValue=<rgba 0.3646 0.6478 1 1>}.RGB
   Alpha=Saturate(None=Add(A=@n9, B=@n13))
  )
  n15 = LinearInterpolate(A=@n14, Alpha=NamedRerouteUsage()) {constB=0}
  n16 = Multiply(A=@n15) {constB=0.6}
  NamedRerouteDeclaration(None=Multiply(A=MaterialFunctionCall(Color_Morning=@n16, Color_Day=@n15, Color_Evening=@n15, Color_Night=@n16) {materialFunction=&Color_Time_of_Day}.Result) {constB=2}) {name=EmissiveColor, nodeColor=<rgba 0 1 0.2734 1>}
  n17 = Multiply(A=NamedRerouteUsage(), B=@n14)
  n18 = NamedRerouteUsage()
  n19 = TextureSample(UVs=Multiply(A=ComponentMask(None=@n18) {b=true, r=true}) {constB=0.5}) {texture=&T_Perlin_Noise_M1}
  n20 = InverseLinearInterpolate(Value=SmoothStep(Value=NamedRerouteUsage()) {constMin=0.2}) {constA=0.35, constB=0.9}
  n21 = Saturate(None=SmoothStep(Min=NamedRerouteDeclaration(None=@n19.G) {name=PerlinNoise, nodeColor=<rgba 1 0 0.5391 1>}, Value=OneMinus(None=@n20)))
  n22 = Multiply(
   A=NamedRerouteDeclaration(None=Multiply(A=@n21) {constB=0.5}) {name=SnowEdge, nodeColor=<rgba 0 0.5078 1 1>}
   B=VectorParameter() {parameterName=SnowColor, defaultValue=<rgba 0.0722 0.5536 1 1>}.RGB
  ) {constB=0.75}
  NamedRerouteDeclaration(None=Desaturation(None=Add(A=@n17, B=@n22), Fraction=Constant() {r=0.25})) {name=BaseColor, nodeColor=<rgba 1 0 0.0469 1>}
  n23 = NamedRerouteUsage()
  n24 = DeriveNormalZ(InXY=ConstantBiasScale(None=NamedRerouteUsage()) {bias=-0.5, scale=2})
  n25 = Constant3Vector() {constant=<rgba 0 0 1 0>}
  n26 = NamedRerouteDeclaration(None=Fresnel(Normal=VertexNormalWS()) {baseReflectFraction=0, exponent=1}) {name=Fresnel, nodeColor=<rgba 0 1 0.3906 1>}
  n27 = Saturate(None=Add(A=Multiply(A=@n26) {constB=2}, B=NamedRerouteUsage()))
  n28 = MaterialFunctionCall(
   Normal=NamedRerouteDeclaration(None=LinearInterpolate(A=@n24, B=@n25, Alpha=@n27)) {name=FrostNormal, nodeColor=<rgba 0.8984 1 0 1>}
   Flatness=Constant() {r=0.5}
  ) {materialFunction=&FlattenNormal}
  NamedRerouteDeclaration(None=LinearInterpolate(
    A=@n23
    B=MaterialFunctionCall(BaseNormal=@n23, AdditionalNormal=@n28.Result) {materialFunction=&BlendAngleCorrectedNormals}.Result
    Alpha=NamedRerouteUsage()
   )) {name=OutNormal, nodeColor=<rgba 0 0.3672 1 1>}
  NamedRerouteDeclaration(None=@n1.Tech Art Mask) {name=FXMask, nodeColor=<rgba 1 0 0.7734 1>}
  n29 = NamedRerouteUsage()
  n30 = SetMaterialAttributes!(
   MaterialAttributes=NamedRerouteUsage()
   Base Color=NamedRerouteUsage()
   Emissive Color=NamedRerouteUsage()
   Metallic=Constant()
   Roughness=Constant() {r=0.1}
   Normal=NamedRerouteUsage()
   Specular=Constant() {r=0.75}
  ) {attributeSetTypes [
    #1
    #3
    #4
    #5
    #2
    #6
   ]}
  n31 = Multiply(
   A=ChannelMaskParameter(None=NamedRerouteUsage()) {maskChannel=Green, parameterName=ReactivityFrostMaskChannel, defaultValue=<rgba 0 1 0 0>}.R
   B=Multiply(A=NamedRerouteUsage(), B=ScalarParameter() {defaultValue=1, parameterName=FrostColorBlend})
  )
  NamedRerouteDeclaration(None=StaticSwitchParameter(True=BlendMaterialAttributes(A=@n29, B=@n30, Alpha=@n31), False=NamedRerouteUsage()) {defaultValue=true, parameterName=UseReactivity}) {name=OutMaterialAttributes, nodeColor=<rgba 0 1 0.9063 1>}
  n32 = Multiply(A=NamedRerouteUsage())
  n33 = NamedRerouteDeclaration(None=Saturate(None=ScalarParameter() {parameterName=Velocity})) {name=InVelocity, nodeColor=<rgba 0 0.4844 1 1>}
  n34 = ScalarParameter() {defaultValue=1, parameterName=ReactivityOn}
  n35 = SphereMask(A=NamedRerouteUsage(), B=Constant()) {attenuationRadius=0.4, hardnessPercent=0}
  n36 = SmoothStep(Value=InverseLinearInterpolate(A=Multiply(A=@n33, B=@n34), Value=@n35) {constA=0.9})
  n37 = OneMinus(None=Saturate(None=SmoothStep(Max=@n32, Value=@n36)))
  n38 = Multiply(A=@n37, B=@n37)
  NamedRerouteDeclaration(None=Saturate(None=Multiply(A=@n38, B=@n33))) {name=FrostMask, nodeColor=<rgba 0 1 0.7422 1>}
  NamedRerouteDeclaration(None=@n4.Normal) {name=InNormal, nodeColor=<rgba 0 0.4141 1 1>}
  NamedRerouteDeclaration(None=Saturate(None=Power(Base=@n6.B) {constExponent=4})) {name=FrostTexture, nodeColor=<rgba 0.8047 1 0 1>}
  NamedRerouteDeclaration(None=@n35) {name=SphereMask, nodeColor=<rgba 0 1 0.9297 1>}
 ]
}
